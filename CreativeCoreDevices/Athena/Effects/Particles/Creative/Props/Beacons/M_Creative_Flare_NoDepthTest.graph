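# Unlit alpha-composite sprite particle material for a Creative flare, likely for beacon props, that draws without depth testing.
# On deferred shading, scene versus pixel depth blends front and behind opacity and texture, desaturates the occluded look, and feeds a distance fade.
# Paired sines on per-particle time drive flicker in exponent and size.
# Vertex offset also scales sprites with camera distance and collapses faded ones; textures support U-indexed atlases or a radial gradient.
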Material /CreativeCoreDevices/Athena/Effects/Particles/Creative/Props/Beacons/M_Creative_Flare_NoDepthTest
graph {
 properties {
  allowTranslucentCustomDepthWrites=true
  bDisableDepthTest=true
  bEnableResponsiveAA=true
  bUsedWithParticleSprites=true
  blendMode=BLEND_AlphaComposite
  opacityMaskClipValue=0.02
  pixelDepthOffsetMode=PDOM_Legacy
  shadingModel=MSM_Unlit
  usedShadingModels=Unlit
 }
 nodes [
  n1 = ScalarParameter() {defaultValue=1, parameterName=OpacityBehind}
  n2 = ScalarParameter() {defaultValue=1, parameterName=OpacityFront}
  n3 = SceneDepth()
  n4 = PixelDepth()
  n5 = ShadingPathSwitch(Default=Constant() {r=1}, Deferred=Saturate(None=Subtract(A=@n3.R, B=@n4.R)))
  n6 = ShadingPathSwitch(
   Default=Constant() {r=1}
   Deferred=MaterialFunctionCall(
    Opacity In=LinearInterpolate(A=@n1, B=@n2, Alpha=@n5)
    Position=ParticlePositionWS()
    Fade_Start_Distance=ScalarParameter() {defaultValue=500, parameterName="Fade_Start_Distance (S)"}
    Fade_End_Distance=ScalarParameter() {defaultValue=300, parameterName="Fade_End_Distance (S)"}
   ) {materialFunction=&MF_Distance_SyncedFade_FX}.Result
  )
  n7 = ScalarParameter() {defaultValue=10, parameterName=WPO}
  n8 = Multiply(A=@n7) {constB=-1}
  n9 = Multiply(A=Time(), B=DynamicParameter() {paramNames [TimeRandom,U_Index,Param3,Param4]}.TimeRandom)
  n10 = ConstantBiasScale(None=Add(A=Sine(None=@n9) {period=0.1635}, B=Sine(None=@n9) {period=0.2886})) {bias=2, scale=0.25}
  n11 = Subtract(A=ParticlePositionWS(), B=WorldPosition().XYZ)
  n12 = WorldPosition()
  n13 = MaterialFunctionCall() {materialFunction=&ObjectPivotPoint}
  n14 = Distance(
   A=MaterialFunctionCall() {materialFunction=&ObjectPivotPoint}.Object Pivot Location
   B=CameraPositionWS()
  )
  n15 = ScalarParameter() {defaultValue=2000, parameterName=ScaleByDistanceDivisor}
  n16 = Add(
   A=Multiply(A=LinearInterpolate(A=@n8, B=@n7, Alpha=@n10) {constB=100}, B=Normalize(VectorInput=@n11))
   B=Multiply(A=Subtract(A=@n12.XYZ, B=@n13.Object Pivot Location), B=Divide(A=@n14, B=@n15) {constB=1000})
  )
  If(A=@n6, B=Constant() {r=0.02}, "A > B"=@n16, "A == B"=@n16, "A < B"=@n11)
  n17 = ParticleColor()
  n18 = SphereMask(
   A=TextureCoordinate()
   B=ScalarParameter() {defaultValue=0.5, parameterName=CenterFront}
   Radius=ScalarParameter() {defaultValue=0.5, parameterName=RadiusFront}
   Hardness=ScalarParameter() {parameterName=HardnessFront}
  )
  n19 = Divide(A=Constant() {r=1}, B=ScalarParameter() {defaultValue=1, parameterName=U_Divisions})
  n20 = Constant() {r=1}
  n21 = DynamicParameter() {paramNames [TimeRandom,U_Index,Param3,Param4]}
  n22 = Add(A=Multiply(A=TextureCoordinate(), B=AppendVector(A=@n19, B=@n20)), B=AppendVector(A=Multiply(A=@n21.U_Index, B=@n19), B=Constant() {r=1}))
  n23 = TextureSampleParameter2D!(UVs=@n22, Level, Apply View MipBias) {mipValueMode=TMVM_MipLevel, parameterName=TextureBehind, texture=/Game/Effects/Fort_Effects/Textures/Flares/T_FlareBehindTranslucency}
  n24 = TextureSampleParameter2D(UVs=@n22) {
   mipValueMode=TMVM_MipLevel
   parameterName=TextureFront
   samplerType=SAMPLERTYPE_Grayscale
   texture=/Game/Effects/Fort_Effects/Textures/Flares/T_Cross_Flare
  }
  n25 = LinearInterpolate(
   A=DotProduct(A=@n23.RGB, B=VectorParameter() {parameterName=TextureBehindChannelPicker, defaultValue=<rgba 1 0 0 0>}.RGB)
   B=DotProduct(A=@n24.RGB, B=VectorParameter() {parameterName=TextureFrontChannelPicker, defaultValue=<rgba 1 0 0 0>}.RGB)
   Alpha=@n5
  )
  n26 = ScalarParameter() {defaultValue=1.5, parameterName=TextureExponentBehind}
  n27 = ScalarParameter() {defaultValue=2.25, parameterName=TextureExponentFront}
  n28 = Multiply(
   A=Power(Base=StaticSwitchParameter(True=@n18, False=@n25) {parameterName=UseRadialGradient}, Exp=LinearInterpolate(A=@n26, B=@n27, Alpha=@n10))
   B=Multiply(A=@n6, B=ParticleColor().A)
  )
  n29 = OneMinus(None=@n5)
  n30 = ScalarParameter() {parameterName=DesaturateBehindFractionMin}
  n31 = ScalarParameter() {defaultValue=0.7, parameterName=DesaturateBehindFractionMax}
  MaterialFunctionCall(Emissive=Desaturation(None=Multiply(A=@n17.RGB, B=@n28), Fraction=Clamp(None=@n29, Min=@n30, Max=@n31) {maxDefault=0.7})) {materialFunction=&MF_FX_EmissiveAdjustments}
 ]
}
